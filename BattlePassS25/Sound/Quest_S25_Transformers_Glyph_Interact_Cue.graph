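# Sound cue for interacting with a glyph in the Season 25 Battle Pass Transformers quest, judging by its name.
# A single wave player feeds the root directly with the Quest_S25_Transformers_Glyph_Interact sound wave, with no randomization or modulation nodes.
# The cue plays at reduced volume through the Fort_Quests sound class and sets no attenuation asset.
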SoundCue /BattlePassS25/Sound/Quest_S25_Transformers_Glyph_Interact_Cue
root=WavePlayer_0
nodes [{id=WavePlayer_0, properties {soundWaveAssetPtr=&/Quest_S25_Transformers_Glyph_Interact}}]
properties {maxDistance=2097152, soundClassObject=&Fort_Quests, volumeMultiplier=0.65}
