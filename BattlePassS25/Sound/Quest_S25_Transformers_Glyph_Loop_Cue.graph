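# Sound cue for a looping Transformers glyph quest sound in the Season 25 Battle Pass content.
# A looping wave player feeds a modulator that slightly randomizes and lowers volume.
# The cue routes to the quests sound class and borrows the supply-drop hidden star ambient loop attenuation for spatial falloff.
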SoundCue /BattlePassS25/Sound/Quest_S25_Transformers_Glyph_Loop_Cue
root=Modulator_2
nodes [
 {id=Modulator_2, children [WavePlayer_0], properties {volumeMax=0.7, volumeMin=0.6}}
 {id=WavePlayer_0, properties {bLooping=true, soundWaveAssetPtr=&/Quest_S25_Transformers_Glyph_Loop}}
]
properties {
 attenuationSettings=/Game/Athena/Sounds/SupplyDrops/Fort_HiddenStar_Ambient_Loop_Att
 maxDistance=2500
 soundClassObject=&Fort_Quests
 volumeMultiplier=1.2
}
