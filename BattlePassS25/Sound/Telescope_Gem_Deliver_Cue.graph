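# Sound cue for a telescope gem delivery sound in the BattlePassS25 content, routed through the non-critical in-game UI sound class.
# A single wave player feeds Telescope_Gem_Deliver_01 straight to the output, with no randomization or mixing.
# Cue-level attenuation overrides set falloff distance and stereo spread.
# Concurrency overrides allow one instance at a time, stopping the oldest when a new one starts.
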SoundCue /BattlePassS25/Sound/Telescope_Gem_Deliver_Cue
root=WavePlayer_1
nodes [{id=WavePlayer_1, properties {soundWaveAssetPtr=&/Telescope_Gem_Deliver_01}}]
properties {
 bOverrideAttenuation=true
 bOverrideConcurrency=true
 maxDistance=1250
 soundClassObject=&Fort_UI_InGame_NonCritical
 volumeMultiplier=1
 attenuationOverrides {falloffDistance=850, stereoSpread=2300}
 concurrencyOverrides {maxCount=1, resolutionRule=StopOldest}
}
