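# Looping sound cue for a telescope gem in Battle Pass S25, classed under the Fort_Quests sound class.
# A looping wave player feeds a modulator that randomizes volume within a narrow, reduced range, with an overall cue volume multiplier on top.
# Spatial falloff references the shared Hidden Star ambient loop attenuation asset from SupplyDrops, and the cue also carries its own attenuation override values.
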SoundCue /BattlePassS25/Sound/Telescope_Gem_Loop_Cue
root=Modulator_2
nodes [
 {id=Modulator_2, children [WavePlayer_2], properties {volumeMax=0.7, volumeMin=0.6}}
 {id=WavePlayer_2, properties {bLooping=true, soundWaveAssetPtr=&/Telescope_Gem_Loop_01}}
]
properties {
 attenuationSettings=/Game/Athena/Sounds/SupplyDrops/Fort_HiddenStar_Ambient_Loop_Att
 maxDistance=2500
 soundClassObject=&Fort_Quests
 volumeMultiplier=1.12
 attenuationOverrides {falloffDistance=700, attenuationShapeExtents {x=100}}
}
